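# Niagara explosion effect for the Figment Snowballer grenade launcher: an icy blue-white snow burst.
# Initial_Burst drives expanding mesh shells through per-renderer color and scale attributes plus a fading blue light, alongside mesh streaks and smoke sprites.
# An unrendered MeshCloudsSource throws particles upward that MeshClouds reads to spawn cloud trails, plus a burst of its own.
# Snow chunks, snowball meshes and noise-driven GPU snowfall sprites launch upward, then fall under downward acceleration.
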NiagaraSystem /Figment_Launcher_Grenade_Athena/Snowballer_FX/NS_Figment_Snowballer_Explosion_S07
dependencies {
 usedDataInterfaces [
  &NiagaraDataInterfaceParticleRead
  &NiagaraDataInterfaceCurve
  &NiagaraDataInterfaceColorCurve
  &NiagaraDataInterfaceVectorCurve
  &NiagaraDataInterfaceVector2DCurve
  &NiagaraDataInterfaceSpriteRendererInfo
 ]
 usedDynamicInputs []
 usedModules [&SystemState]
}
emitters {
 Initial_Burst {
  emitterUpdateScript [
   &EmitterState()
   &SpawnBurst_Instantaneous(Use Spawn Probability=false, UseLoopCountLimit=false)
   &SpawnBurst_Instantaneous(Spawn Group=1, Use Spawn Probability=false, UseLoopCountLimit=false) {module=SpawnBurst_Instantaneous001}
   &SpawnBurst_Instantaneous(Spawn Group=2, Use Spawn Probability=false, UseLoopCountLimit=false) {module=SpawnBurst_Instantaneous002}
   &SpawnBurst_Instantaneous(Spawn Group=3, Use Spawn Probability=false, UseLoopCountLimit=false) {module=SpawnBurst_Instantaneous003}
  ]
  particleSpawnScript [
   /SetVariables(Particles.VisibilityTag=&SpawnGroup()) {isSetParameters=true, module=SetVariables_#1}
   &InitializeParticle(
    Lifetime=&MaskFloatBySpawnGroup(
     Masked Float=1
     Passthrough Float=&MaskFloatBySpawnGroup(
      Masked Float=0.45
      Passthrough Float=&MaskFloatBySpawnGroup(Masked Float=0.4, Passthrough Float=0.2, Spawn Group Mask=2)
      Spawn Group Mask=1
     )
    )
    Mesh Scale Mode=::NewEnumerator1
    Mesh Uniform Scale Max=10.5
    Mesh Uniform Scale Min=9
    Position Offset=<0 0 20>
    UsePositionOffset=true
    Write Material Random=true
   )
   &InitialMeshOrientation(
    Mesh Orientation Mode=::NewEnumerator1
    Rotation=&RandomRangeVector(Minimum=<0 0 0>, Maximum=<0 0 1>)
    Use Rotation Vector=true
   )
  ]
  particleUpdateScript [
   &ParticleState()
   &ScaleColor(
    Scale Alpha=&FloatFromCurve(FloatCurve=DataInterface() {propertyValues {Curve {keys [<at 0.3348 1 RCIM_Cubic>, {arriveTangent=-1.5033, interpMode=RCIM_Cubic, leaveTangent=-1.5033, time=0.7494, value=0.4852}, <at 1 0>]}}})
    Scale RGB=&VectorFromCurve(VectorCurve=DataInterface() {propertyValues {
       XCurve {keys [{arriveTangent=-2.98, time=0, value=0.377}, {arriveTangent=-0.02, leaveTangent=-0.02, time=0.9, value=0.2043}, <at 1 0.2134 RCIM_Cubic>]}
       YCurve {keys [{arriveTangent=-0.9545, time=0, value=2.3357}, {arriveTangent=-0.005, leaveTangent=-0.005, time=0.9, value=0.8151}, <at 1 0.6763 RCIM_Cubic>]}
       ZCurve {keys [{arriveTangent=-0.2149, time=0, value=3}, {arriveTangent=-0.0251, leaveTangent=-0.0251, time=0.9, value=0.9479}, <at 1 0.9062 RCIM_Cubic>]}
      }})
    ScaleA=true
    ScaleRGB=true
   )
   &ScaleMeshSize(Scale Factor=&VectorFromFloat(Value=&FloatFromCurve(FloatCurve=DataInterface() {propertyValues {Curve {keys [
          {arriveTangent=6.3495, interpMode=RCIM_Cubic, leaveTangent=6.3495, tangentMode=RCTM_User, time=0, value=0}
          {arriveTangent=1.2114, interpMode=RCIM_Cubic, leaveTangent=1.2114, time=0.2381, value=0.9424}
          <at 1 1.2114>
         ]}}})))
   &DynamicMaterialParameters(
    Index 0 Param 1=&FloatFromCurve(FloatCurve=DataInterface() {propertyValues {Curve {keys [{arriveTangent=-1.1363, interpMode=RCIM_Cubic, leaveTangent=-1.1363, tangentMode=RCTM_User, time=0, value=1}, <at 0.25 0.1 RCIM_Cubic>]}}})
    Index 0 Param 2=&FloatFromCurve(FloatCurve=DataInterface() {propertyValues {Curve {keys [
         <at 0 10 RCIM_Cubic>
         {arriveTangent=1.1935, interpMode=RCIM_Cubic, leaveTangent=1.1935, tangentMode=RCTM_User, time=0.2, value=0}
         {arriveTangent=2.5, interpMode=RCIM_Cubic, leaveTangent=2.5, time=0.4937, value=0.4037}
         <at 1 2 RCIM_Cubic>
        ]}}})
    Index 0 Param 4=&FloatFromCurve(FloatCurve=DataInterface() {propertyValues {Curve {keys [
         {arriveTangent=2.897, interpMode=RCIM_Cubic, leaveTangent=2.897, tangentMode=RCTM_User, time=0, value=0}
         {arriveTangent=0.1228, leaveTangent=0.1228, time=1, value=1}
        ]}}})
    Index 1 Param 1=&FloatFromCurve(FloatCurve=DataInterface() {propertyValues {Curve {keys [{arriveTangent=37.424, interpMode=RCIM_Cubic, leaveTangent=37.424, tangentMode=RCTM_User, time=0, value=0}, <at 1 11 RCIM_Cubic>]}}})
    Index 1 Param 2=1
    Index 1 Param 3=1
    Index 1 Param 4=1
    Module.Index0Float0Write=true
    Module.Index0Float1Write=true
    Module.Index0Float2Write=true
    Module.Index0Float3Write=true
    Module.Index1Float0Write=true
    Module.Index1Float1Write=true
    Module.Index1Float2Write=true
    Module.Index1Float3Write=true
    Write Parameter Index 1=::NewEnumerator1
   )
   /SetVariables(
    Particles.SphereMeshColor:LinearColor=&MultiplyLinearColor(
     Linear Color=<rgba 1 1 1 0.3>
     Scale Factor=&MakeVector4FromVectorAndFloat(
      "Vector (RGB)"=&MakeVectorFromLinearColorRGB(Color=&ColorFromCurve(ColorCurve=DataInterface() {propertyValues {
          AlphaCurve {keys [<at 0 1 RCIM_Cubic>]}
          BlueCurve {keys [<at 0.0816 150 RCIM_Cubic>, {arriveTangent=-157.3603, time=0.126, value=5}, <at 0.9919 1>]}
          GreenCurve {keys [<at 0.0816 110.622 RCIM_Cubic>, {arriveTangent=-115.8794, time=0.126, value=4.4954}, <at 0.9919 0.8991>]}
          RedCurve {keys [<at 0.0816 0 RCIM_Cubic>, {arriveTangent=0.6768, time=0.126, value=3.2041}, <at 0.9919 0.6408>]}
         }}))
      "Float (Alpha)"=&FloatFromCurve(FloatCurve=DataInterface() {propertyValues {Curve {keys [
           <at 0 0 RCIM_Cubic>
           {arriveTangent=7.2119, interpMode=RCIM_Cubic, leaveTangent=7.2119, tangentMode=RCTM_User, time=0.0476, value=0.2262}
           {arriveTangent=-1.0231, interpMode=RCIM_Cubic, leaveTangent=-1.0231, tangentMode=RCTM_User, time=0.3638, value=0.3159}
           {arriveTangent=-0.264, interpMode=RCIM_Cubic, leaveTangent=-0.264, tangentMode=RCTM_User, time=1, value=0}
          ]}}})
     )
    )
    Particles.SphereMeshScale:Vector3f=&Multiply_VectorByFloat(
     Vector=<16 16 16>
     Float=&FloatFromCurve(FloatCurve=DataInterface() {propertyValues {Curve {keys [
          {arriveTangent=1.6618, interpMode=RCIM_Cubic, leaveTangent=1.6618, tangentMode=RCTM_User, time=0, value=0}
          {arriveTangent=0.1684, interpMode=RCIM_Cubic, leaveTangent=0.1684, tangentMode=RCTM_User, time=1, value=1}
         ]}}})
    )
   ) {isSetParameters=true}
   /SetVariables(
    Particles.BurstSphereColor:LinearColor=&ColorFromCurve(ColorCurve=DataInterface() {propertyValues {
       AlphaCurve {keys [<at 0 1 RCIM_Cubic>]}
       BlueCurve {keys [<at 0.4074 3 RCIM_Cubic>,<at 0.9727 0.25 RCIM_Cubic>]}
       GreenCurve {keys [<at 0.4074 1.4311 RCIM_Cubic>,<at 0.9727 0.1193 RCIM_Cubic>]}
       RedCurve {keys [<at 0.4074 0.0781 RCIM_Cubic>,<at 0.9727 0.0065 RCIM_Cubic>]}
      }})
    Particles.BurstSphereScale:Vector3f=&Multiply_VectorByFloat(
     Vector=<13 13 13>
     Float=&FloatFromCurve(FloatCurve=DataInterface() {propertyValues {Curve {keys [{arriveTangent=1.4794, interpMode=RCIM_Cubic, leaveTangent=1.4794, tangentMode=RCTM_User, time=0, value=0}, <at 1 1>]}}})
    )
   ) {isSetParameters=true}
   &FN_Light_Attributes(
    Color=<rgba 0.0722 0.554 1 1>
    Create Color Attribute=true
    Create Exponent Attribute=true
    Create Position Attribute=true
    Create Radius Attribute=true
    Exponent=2
    Position=<0 0 50>
    Radius=750
    Scale Color=&FloatFromCurve(
     FloatCurve=DataInterface() {propertyValues {Curve {
        defaultValue=0
        keys [{interpMode=RCIM_Cubic, tangentMode=RCTM_User, time=0, value=1}, {interpMode=RCIM_Cubic, tangentMode=RCTM_User, time=1, value=0}]
       }}}
     Scale Curve=120
    )
   )
  ]
  rendererClasses [&NiagaraMeshRendererProperties, &NiagaraLightRendererProperties]
  renderers [
   {rendererClass=&NiagaraMeshRendererProperties, rendererIndex=0}
   {rendererClass=&NiagaraMeshRendererProperties, rendererIndex=1}
   {rendererClass=&NiagaraMeshRendererProperties, rendererIndex=2}
   {rendererClass=&NiagaraLightRendererProperties, rendererIndex=3}
  ]
 }
 MeshClouds {
  emitterUpdateScript [
   &EmitterState(Life Cycle Mode=::NewEnumerator1, Loop Behavior=::NewEnumerator1, Loop Delay=0.09, Loop Duration=0.55, UseLoopDelay=true)
   &SpawnParticlesFromOtherEmitter(
    Attribute Reader=DataInterface() {propertyValues {EmitterBinding {bindingMode=Other, emitterName=MeshCloudsSource}}}
    Spawn Rate=20
    Spawn Rate Per Particle Cap=225
    Use Maximum Age=false
    Use Minimum Age=false
   )
   &SpawnBurst_Instantaneous(
    Spawn Count=&RandomRangeInt(Minimum=15, Maximum=20)
    Spawn Group=1
    Spawn Time=0.05
    Use Spawn Probability=false
    UseLoopCountLimit=false
   )
  ]
  particleSpawnScript [
   &InitializeParticle(
    Lifetime Max=1
    Lifetime Min=0.3
    Lifetime Mode=::NewEnumerator1
    Mesh Scale Mode=::NewEnumerator3
    Mesh Uniform Scale=&MaskFloatBySpawnGroup(
     Masked Float=&FloatFromCurve(
      FloatCurve=DataInterface() {propertyValues {Curve {keys [{arriveTangent=-1, time=0.1971, value=3}, <at 0.4 2>]}}}
      CurveIndex=$Emitter.NormalizedLoopAge
     )
     Passthrough Float=&RandomRangeFloat(Minimum=1.875, Maximum=3, Evaluation Type=::NewEnumerator1)
    )
    UsePositionOffset=false
    Write Material Random=true
   )
   &SampleParticlesFromOtherEmitter()
   /SetVariables(Particles.Position=&MaskVectorBySpawnGroup(
     Masked Vector=Linked() {linkedVariable {name=Output.SampleParticlesFromOtherEmitter.SampledPosition, type=NiagaraPosition}}
     Passthrough Vector=Linked() {linkedVariable {name=Engine.Emitter.SimulationPosition, type=NiagaraPosition}}
    )) {isSetParameters=true}
   &ShapeLocation(
    Cylinder Height=80
    Cylinder Radius=&RandomRangeFloat(Minimum=16, Maximum=25, Evaluation Type=::NewEnumerator1)
    MaskSpawnGroup=false
    Offset=<0 0 20>
    Offset Mode=::NewEnumerator0
    Override Seed=false
    Rotation Mode=::NewEnumerator2
    Shape Primitive=::NewEnumerator1
    Surface Only=false
    UseOwnerScale=true
   )
   &ShapeLocation(MaskSpawnGroup=false, Override Seed=false, Sphere Radius=20, UseOwnerScale=true) {module=ShapeLocation001}
   &AddVelocity(Velocity=&MaskVector(
     Mask Source=0
     Mask Parameter=?
     Mask=0
     Masked Vector=&Multiply_VectorByFloat(Vector=$Particles.ShapeLocation001.ShapeVector, Float=2)
     Passthrough Vector=&Add_Vector(
      A=&RandomRangeVector(Minimum=<20 0 0>, Maximum=<150 0 0>, Evaluation Type=::NewEnumerator1)
      B=&Multiply_VectorByFloat(Vector=&Multiply_Vector(A=$Particles.ShapeLocation.ShapeVector, B=<1 1 0>), Float=50)
     )
    ))
   &InitialMeshOrientation(Use Rotation Vector=false)
  ]
  particleUpdateScript [
   &ParticleState()
   &Drag(Drag=3, Use Linear Drag=true, Use Rotational Drag=true)
   &AccelerationForce(Acceleration=&MaskVector(
     Mask Source=0
     Mask Parameter=?
     Mask=0
     Masked Vector=&RandomRangeVector(Minimum=<0 0 -250>, Maximum=<0 0 -800>, Evaluation Type=::NewEnumerator1)
     Passthrough Vector=&RandomRangeVector(Minimum=<0 0 20>, Maximum=<0 0 500>)
    ))
   &SolveForcesAndVelocity(Clamp Velocity=false, Limit Acceleration=false)
   &ScaleMeshSize(Scale Factor=&VectorFromFloat(Value=&MaskFloat(
      Mask Source=0
      Mask Parameter=?
      Mask=0
      Masked Float=&FloatFromCurve(
       FloatCurve=DataInterface() {propertyValues {Curve {keys [
           {arriveTangent=4.8845, interpMode=RCIM_Cubic, leaveTangent=4.8845, tangentMode=RCTM_User, time=-0.0009, value=0.1398}
           {arriveTangent=1.2355, interpMode=RCIM_Cubic, leaveTangent=1.2355, time=0.2074, value=0.8048}
           {arriveTangent=-1.0154, interpMode=RCIM_Cubic, leaveTangent=-1.0154, time=0.6954, value=1}
           <at 1 0>
          ]}}}
       Scale Curve=1.5
      )
      Passthrough Float=&FloatFromCurve(
       FloatCurve=DataInterface() {propertyValues {Curve {keys [
           {arriveTangent=3.3815, interpMode=RCIM_Cubic, leaveTangent=3.3815, tangentMode=RCTM_User, time=-0.0009, value=0.0199}
           {arriveTangent=1.2424, interpMode=RCIM_Cubic, leaveTangent=1.2424, time=0.2163, value=0.8373}
           <at 0.9402 1.1891 RCIM_Cubic>
          ]}}}
       Scale Curve=1.5
      )
     )))
   &DynamicMaterialParameters(
    Index 0 Param 1=&FloatFromCurve(FloatCurve=DataInterface() {propertyValues {Curve {keys [{arriveTangent=-4.3253, interpMode=RCIM_Cubic, leaveTangent=-4.3253, tangentMode=RCTM_User, time=0.226, value=1}, <at 0.7 0.1>]}}})
    Index 0 Param 2=&FloatFromCurve(FloatCurve=DataInterface() {propertyValues {Curve {keys [<at 0.095 0 RCIM_Cubic>, {interpMode=RCIM_Cubic, leaveTangent=-15.55, time=1, value=3}]}}})
    Index 0 Param 3=1
    Index 0 Param 4=&FloatFromCurve(FloatCurve=DataInterface() {propertyValues {Curve {keys [<at 0.3 0>,<at 0.561 1>]}}})
    Module.Index0Float0Write=true
    Module.Index0Float1Write=true
    Module.Index0Float2Write=true
    Module.Index0Float3Write=true
   )
   &UpdateMeshOrientation(Rotation Rate=-0.35)
   &ScaleColor(
    Scale Alpha=&FloatFromCurve(FloatCurve=DataInterface() {propertyValues {Curve {keys [<at 0.4474 1 RCIM_Cubic>,<at 0.6445 1>,<at 1 0>]}}})
    Scale RGB=&MakeVectorFromLinearColorRGB(Color=&ColorFromCurve(ColorCurve=DataInterface() {propertyValues {
        AlphaCurve {keys [<at 0 1 RCIM_Cubic>]}
        BlueCurve {keys [<at 0.26 1 RCIM_Cubic>,<at 0.47 1>,<at 0.83 3 RCIM_Cubic>]}
        GreenCurve {keys [<at 0.26 0.8796 RCIM_Cubic>,<at 0.47 0.9387>,<at 0.83 1.0134 RCIM_Cubic>]}
        RedCurve {keys [<at 0.26 0 RCIM_Cubic>,<at 0.47 0.8879>,<at 0.83 0 RCIM_Cubic>]}
       }}))
    ScaleA=true
    ScaleRGB=true
   )
  ]
  rendererClasses [&NiagaraMeshRendererProperties]
  renderers [{rendererClass=&NiagaraMeshRendererProperties, rendererIndex=0}]
 }
 MeshCloudsSource {
  emitterUpdateScript [
   &EmitterState()
   &SpawnBurst_Instantaneous(Spawn Count=15, Spawn Time=0.1, Use Spawn Probability=false, UseLoopCountLimit=false)
  ]
  particleSpawnScript [
   &InitializeParticle(Color Mode=::NewEnumerator0, Lifetime=0.6, UsePositionOffset=false, Write Material Random=true)
   &ShapeLocation(Hemisphere Distribution=<1 0.5>, MaskSpawnGroup=false, Override Seed=false, Sphere Radius=50, UseOwnerScale=true)
   &AddVelocity(
    Cone Angle=&RandomRangeFloat(Minimum=0.05, Maximum=0.1, Evaluation Type=::NewEnumerator1)
    Cone Angle Mode:ENiagara_AngleInput=::NewEnumerator1
    Cone Axis=<0 0 1>
    Module.UseConeFalloff=false
    Module.UseVelDistribution=false
    Override Seed=false
    Velocity Mode=::NewEnumerator2
    Velocity Speed=&RandomRangeFloat(Minimum=1800, Maximum=6200, Evaluation Type=::NewEnumerator1)
   )
  ]
  particleUpdateScript [
   &ParticleState()
   &Drag(
    Drag=&RandomRangeFloat(Minimum=3, Maximum=4, Evaluation Type=::NewEnumerator1)
    Use Linear Drag=true
    Use Rotational Drag=true
   )
   &AccelerationForce(Acceleration=&RandomRangeVector(Minimum=<0 0 -1000>, Maximum=<0 0 -6000>, Evaluation Type=::NewEnumerator1))
   &SolveForcesAndVelocity(Clamp Velocity=false, Limit Acceleration=false)
  ]
  rendererClasses []
  renderers []
 }
 Smoke {
  emitterUpdateScript [
   &EmitterState()
   &SpawnBurst_Instantaneous(Spawn Count=4, Use Spawn Probability=false, UseLoopCountLimit=false)
   &SpawnBurst_Instantaneous(Spawn Count=12, Spawn Group=1, Use Spawn Probability=false, UseLoopCountLimit=false) {module=SpawnBurst_Instantaneous001}
  ]
  particleSpawnScript [
   &InitializeParticle(
    Color Maximum=<rgba 1 1 1 0.25>
    Color Minimum=<rgba 1 1 1 0.1>
    Color Mode=::NewEnumerator2
    Lifetime Max=1.35
    Lifetime Min=0.85
    Lifetime Mode=::NewEnumerator1
    Sprite Size Mode=::NewEnumerator3
    Sprite UV Mode=::NewEnumerator1
    Uniform Sprite Size=&MaskFloatBySpawnGroup(
     Masked Float=&RandomRangeFloat(Minimum=700, Maximum=900, Evaluation Type=::NewEnumerator1)
     Passthrough Float=&RandomRangeFloat(Minimum=300, Maximum=500, Evaluation Type=::NewEnumerator1)
    )
    UsePositionOffset=false
    Write Material Random=true
   )
   &ShapeLocation(
    Hemisphere Distribution=<1 0.5>
    MaskSpawnGroup=false
    Offset=<0 0 120>
    Offset Mode=::NewEnumerator0
    Override Seed=false
    Sphere Radius=150
    UseOwnerScale=true
   )
   &ShapeLocation(
    Cylinder Height=80
    Cylinder Radius=&RandomRangeFloat(Minimum=30, Maximum=40, Evaluation Type=::NewEnumerator1)
    MaskSpawnGroup=false
    Offset=<0 0 -50>
    Offset Mode=::NewEnumerator0
    Override Seed=false
    Rotation Mode=::NewEnumerator2
    Shape Primitive=::NewEnumerator1
    Surface Only=false
    UseOwnerScale=true
   ) {module=ShapeLocation001}
   &AddVelocity(Velocity=&MaskVectorBySpawnGroup(
     Masked Vector=&RandomRangeVector(Minimum=<0 0 200>, Maximum=<0 0 1200>, Evaluation Type=::NewEnumerator1)
     Passthrough Vector=&Add_Vector(
      A=&RandomRangeVector(Minimum=<0 0 50>, Maximum=<0 0 200>, Evaluation Type=::NewEnumerator1)
      B=&Multiply_VectorByFloat(Vector=&Multiply_Vector(A=$Particles.ShapeLocation001.ShapeVector, B=<1 1 0>), Float=30)
     )
    ))
  ]
  particleUpdateScript [
   &ParticleState()
   &Drag(
    Drag=&MaskFloatBySpawnGroup(Masked Float=1, Passthrough Float=3.5)
    Use Linear Drag=true
    Use Rotational Drag=true
   )
   &SolveForcesAndVelocity(Clamp Velocity=false, Limit Acceleration=false)
   &ScaleColor(
    Scale Alpha=&FloatFromCurve(FloatCurve=DataInterface() {propertyValues {Curve {keys [<at 0.6 1>,<at 1 0>]}}})
    Scale RGB=&MakeVectorFromLinearColorRGB(Color=&ColorFromCurve(ColorCurve=DataInterface() {propertyValues {
        AlphaCurve {keys [<at 0 1 RCIM_Cubic>,<at 1 0 RCIM_Cubic>]}
        BlueCurve {keys [<at 0.2827 2 RCIM_Cubic>,<at 1 0.4531 RCIM_Cubic>]}
        GreenCurve {keys [<at 0.2827 1.6996 RCIM_Cubic>,<at 1 0.4531 RCIM_Cubic>]}
        RedCurve {keys [<at 0.2827 1.2708 RCIM_Cubic>,<at 1 0.4531 RCIM_Cubic>]}
       }}))
    ScaleA=true
    ScaleRGB=true
   )
   &ScaleSpriteSize(
    Scale Sprite Size Mode=::NewEnumerator0
    Uniform Scale Factor=&FloatFromCurve(FloatCurve=DataInterface() {propertyValues {Curve {keys [
         {arriveTangent=4.9678, interpMode=RCIM_Cubic, leaveTangent=4.9678, tangentMode=RCTM_User, time=0.0008, value=0.5442}
         {arriveTangent=0.4561, interpMode=RCIM_Cubic, leaveTangent=0.4561, time=0.095, value=0.6959}
         <at 1 1>
        ]}}})
   )
   &SubUVAnimation(
    Frames Per Second=16
    Playback Mode:ENiagara_FPSPlayrate=::NewEnumerator1
    "Sprite Renderer "=DataInterface() {propertyValues {SpriteRenderer=&self:Smoke_0.NiagaraSpriteRendererProperties_0}}
    SubUV Animation Mode:ENiagara_SubUVLookupModeV2=::NewEnumerator3
    UseEndFrame=false
    UsePlayRateScale=false
    UseStartFrame=false
   )
  ]
  rendererClasses [&NiagaraSpriteRendererProperties]
  renderers [{rendererClass=&NiagaraSpriteRendererProperties}]
 }
 SnowChunks {
  emitterUpdateScript [
   &EmitterState()
   &SpawnBurst_Instantaneous(
    Spawn Count=&RandomRangeInt(Minimum=50, Maximum=100)
    Use Spawn Probability=false
    UseLoopCountLimit=false
   )
  ]
  particleSpawnScript [
   &InitializeParticle(
    Color=<rgba 5 5 5 0.8>
    Lifetime Max=1.5
    Lifetime Min=0.25
    Lifetime Mode=::NewEnumerator1
    Sprite Size Max=<25 100>
    Sprite Size Min=<5 50>
    Sprite Size Mode=::NewEnumerator2
    UsePositionOffset=false
    Write Material Random=true
   )
   &ShapeLocation(
    Hemisphere Distribution=<1 0.5>
    MaskSpawnGroup=false
    Offset=<0 0 20>
    Offset Mode=::NewEnumerator0
    Override Seed=false
    Sphere Radius=50
    Sphere Surface Distribution=1
    UseOwnerScale=true
   )
   &AddVelocity(
    Rotation Coordinate Space=::World
    Velocity=&Add_Vector(
     A=&RandomRangeVector(Minimum=<0 0 500>, Maximum=<0 0 1900>, Evaluation Type=::NewEnumerator1)
     B=&Multiply_VectorByFloat(
      Vector=$Particles.ShapeLocation.ShapeVector
      Float=&RandomRangeFloat(Minimum=2, Maximum=5, Evaluation Type=::NewEnumerator1)
     )
    )
   )
   &AddVelocity(
    Cone Angle=&RandomRangeFloat(Minimum=0.1, Maximum=0.15, Evaluation Type=::NewEnumerator1)
    Cone Angle Mode:ENiagara_AngleInput=::NewEnumerator1
    Cone Axis=<0 0 1>
    Module.UseConeFalloff=false
    Module.UseVelDistribution=false
    Override Seed=false
    Velocity Mode=::NewEnumerator2
    Velocity Speed=&RandomRangeFloat(Minimum=1000, Maximum=1250, Evaluation Type=::NewEnumerator1)
   ) {module=AddVelocity001}
  ]
  particleUpdateScript [
   &ParticleState()
   &AccelerationForce(Acceleration=<0 0 -2000>)
   &Drag(Drag=1.2, Use Linear Drag=true, Use Rotational Drag=true)
   &SolveForcesAndVelocity(Clamp Velocity=false, Limit Acceleration=false)
   &ScaleSpriteSize(
    Scale Factor=&Vector2DFromCurve(Vector2Curve=DataInterface() {propertyValues {XCurve {keys [<at 0.7 1>,<at 1 0>]}, YCurve {keys [<at 0.7 1>,<at 1 0>]}}}, ScaleCurve=<1 1>)
    Scale Sprite Size Mode=::NewEnumerator2
   )
   &ScaleSpriteSizeBySpeed(
    Max Scale Factor=<1 1>
    Min Scale Factor=<0.25 0.25>
    Source Velocity:Vector3f=&Multiply_Vector(A=<0 0.25 1>, B=$Particles.Velocity)
   )
   &ScaleColor(
    Scale Alpha=&FloatFromCurve(FloatCurve=DataInterface() {propertyValues {Curve {keys [<at 0 0>,<at 0.05 1>,<at 0.95 1>,<at 1 0>]}}})
    Scale RGB=&VectorFromCurve(VectorCurve=DataInterface() {propertyValues {XCurve {keys [<at 0 7.2536>,<at 0.5 1>]}, YCurve {keys [<at 0 19.3472>,<at 0.5 1>]}, ZCurve {keys [<at 0 50>,<at 0.5 1>]}}})
    ScaleA=true
    ScaleRGB=true
   )
  ]
  rendererClasses [&NiagaraSpriteRendererProperties]
  renderers [{rendererClass=&NiagaraSpriteRendererProperties}]
 }
 Snowball_Mesh {
  emitterUpdateScript [
   &EmitterState()
   &SpawnBurst_Instantaneous(
    Spawn Count=&RandomRangeInt(Minimum=25, Maximum=40)
    Spawn Time=0.1
    Use Spawn Probability=false
    UseLoopCountLimit=false
   )
  ]
  particleSpawnScript [
   &InitializeParticle(
    Color=<rgba 1 1 1 0.6>
    Lifetime Min=0.7
    Lifetime Mode=::NewEnumerator1
    Mesh Scale Max=<1 2 2>
    Mesh Scale Min=<0.5 0.5 0.5>
    Mesh Scale Mode=::NewEnumerator2
    UsePositionOffset=false
    Write Material Random=true
   )
   &InitialMeshOrientation(Use Rotation Vector=false)
   &ShapeLocation(
    MaskSpawnGroup=false
    Offset=<0 0 200>
    Offset Mode=::NewEnumerator0
    Override Seed=false
    Rotation Mode=::NewEnumerator2
    Sphere Radius=1
    Sphere Surface Distribution=1
    UseOwnerScale=true
   )
   &AddVelocity(Velocity=&Add_Vector(
     A=&RandomRangeVector(Minimum=<0 0 150>, Maximum=<15 0 500>, Evaluation Type=::NewEnumerator1)
     B=&Multiply_VectorByFloat(Vector=$Particles.ShapeLocation.ShapeVector, Float=5)
    ))
   &AddVelocity(
    Cone Angle=&RandomRangeFloat(Minimum=0, Maximum=0.15, Evaluation Type=::NewEnumerator1)
    Cone Angle Mode:ENiagara_AngleInput=::NewEnumerator1
    Cone Axis=<0 0 1>
    Module.UseConeFalloff=false
    Module.UseVelDistribution=false
    Override Seed=false
    Velocity Mode=::NewEnumerator2
    Velocity Speed=&RandomRangeFloat(Minimum=250, Maximum=1000, Evaluation Type=::NewEnumerator1)
   ) {module=AddVelocity001}
  ]
  particleUpdateScript [
   &ParticleState()
   &AccelerationForce(Acceleration=&RandomRangeVector(Minimum=<0 0 -2000>, Maximum=<0 0 -2500>, Evaluation Type=::NewEnumerator1))
   &SolveForcesAndVelocity(Clamp Velocity=false, Limit Acceleration=false)
   &ScaleMeshSize(Scale Factor=&VectorFromFloat(Value=&FloatFromCurve(
      FloatCurve=DataInterface() {propertyValues {Curve {keys [
          <at -0.002 0.0132 RCIM_Cubic>
          {arriveTangent=-0.0131, interpMode=RCIM_Cubic, leaveTangent=-0.0131, time=0.1279, value=1}
          {arriveTangent=-3.2384, interpMode=RCIM_Cubic, leaveTangent=-3.2384, tangentMode=RCTM_User, time=1, value=0}
         ]}}}
      Scale Curve=0.5
     )))
   &ScaleMeshSizeBySpeed(
    Max Scale Factor=<1.25 1.25 1.25>
    Min Scale Factor:Vector3f=<0.5 0.5 0.5>
    Source Velocity:Vector3f=&Multiply_Vector(A=<0.5 0 1>, B=$Particles.Velocity)
   )
   &DynamicMaterialParameters(
    Index 0 Param 1=&FloatFromCurve(FloatCurve=DataInterface() {propertyValues {Curve {keys [
         {arriveTangent=3.9125, interpMode=RCIM_Cubic, leaveTangent=3.9125, tangentMode=RCTM_User, time=0, value=0}
         <at 0.8 1 RCIM_Cubic>
         <at 1 0 RCIM_Cubic>
        ]}}})
    Index 0 Param 2=1
    Index 0 Param 3=1
    Index 0 Param 4=1
    Module.Index0Float0Write=true
    Module.Index0Float1Write=true
    Module.Index0Float2Write=true
    Module.Index0Float3Write=true
   )
  ]
  rendererClasses [&NiagaraMeshRendererProperties]
  renderers [{rendererClass=&NiagaraMeshRendererProperties, rendererIndex=0}]
 }
 Snowfall_Sprites {
  simTarget=GPUComputeSim
  emitterUpdateScript [
   &EmitterState()
   &SpawnRate(
    SpawnRate=&Multiply_Float(
     A=&RandomRangeFloat(Minimum=1000, Maximum=750, Evaluation Type=::NewEnumerator1)
     B=&FloatFromCurve(FloatCurve=DataInterface() {propertyValues {Curve {keys [<at 0 1>,<at 0.1 1>,<at 0.2 0>]}}}, CurveIndex=$Emitter.NormalizedLoopAge)
    )
    Use Spawn Probability=false
   )
  ]
  particleSpawnScript [
   &InitializeParticle(
    Color=<rgba 1 1 1 1.5>
    Lifetime Max=3
    Lifetime Mode=::NewEnumerator1
    Sprite Rotation Mode=::NewEnumerator2
    Sprite Size Mode=::NewEnumerator1
    Uniform Sprite Size Max=20
    UsePositionOffset=false
    Write Material Random=true
   )
   &ShapeLocation(
    Cylinder Height=&FloatFromCurve(FloatCurve=DataInterface() {propertyValues {Curve {keys [<at 0 0>,<at 0.2 250>]}}}, CurveIndex=$Emitter.LoopedAge)
    Cylinder Radius=&FloatFromCurve(FloatCurve=DataInterface() {propertyValues {Curve {keys [<at 0 5>,<at 0.2 150>]}}}, CurveIndex=$Emitter.LoopedAge)
    MaskSpawnGroup=false
    Offset=<0 0 50>
    Offset Mode=::NewEnumerator0
    Override Seed=false
    Rotation Mode=::NewEnumerator2
    Shape Primitive=::NewEnumerator1
    Surface Only=false
    UseOwnerScale=true
   )
   &AddVelocity(Velocity=&Add_Vector(
     A=&RandomRangeVector(Minimum=<-200 -200 900>, Maximum=<200 200 2000>, Evaluation Type=::NewEnumerator1)
     B=&Multiply_VectorByFloat(
      Vector=$Particles.ShapeLocation.ShapeVector
      Float=&RandomRangeFloat(Minimum=5, Maximum=10, Evaluation Type=::NewEnumerator1)
     )
    ))
  ]
  particleUpdateScript [
   &ParticleState()
   &AccelerationForce(Acceleration=<20 0 -1100>)
   &Drag(
    Drag=&RandomRangeFloat(Minimum=1, Maximum=3, Evaluation Type=::NewEnumerator1)
    Use Linear Drag=true
    Use Rotational Drag=true
   )
   &VectorNoiseForce(Force Amount=6562.5, Module.Use Random Noise Time Interval=false)
   &SolveForcesAndVelocity(Clamp Velocity=false, Limit Acceleration=false)
   &ScaleColor(
    Scale Alpha=&FloatFromCurve(FloatCurve=DataInterface() {propertyValues {Curve {keys [
         {interpMode=RCIM_Cubic, tangentMode=RCTM_User, time=0, value=1}
         {arriveTangent=-1.5396, interpMode=RCIM_Cubic, leaveTangent=-1.5396, tangentMode=RCTM_User, time=1, value=0}
        ]}}})
    Scale RGB=&MakeVectorFromLinearColorRGB(Color=<rgba 1 1 1 1>)
    ScaleA=true
    ScaleRGB=true
   )
   &ScaleSpriteSize(
    Scale Sprite Size Mode=::NewEnumerator0
    Uniform Scale Factor=&FloatFromCurve(FloatCurve=DataInterface() {propertyValues {Curve {keys [{arriveTangent=-3.861, interpMode=RCIM_Cubic, leaveTangent=-3.861, tangentMode=RCTM_User, time=0.5, value=1}, <at 1 0 RCIM_Cubic>]}}})
   )
  ]
  rendererClasses [&NiagaraSpriteRendererProperties]
  renderers [{rendererClass=&NiagaraSpriteRendererProperties}]
 }
 Streaks {
  emitterUpdateScript [
   &EmitterState()
   &SpawnBurst_Instantaneous(
    Spawn Count=&RandomRangeInt(Minimum=2, Maximum=4)
    Use Spawn Probability=false
    UseLoopCountLimit=false
   )
   &SpawnBurst_Instantaneous(
    Spawn Count=&RandomRangeInt(Minimum=2, Maximum=3)
    Spawn Group=1
    Use Spawn Probability=false
    UseLoopCountLimit=false
   ) {module=SpawnBurst_Instantaneous001}
  ]
  particleSpawnScript [
   /SetVariables(Particles.VisibilityTag=&SpawnGroup()) {isSetParameters=true, module=SetVariables_#1}
   &InitializeParticle(
    Color=<rgba 0.7064 0.8088 0.875 4>
    Lifetime=&MaskFloatBySpawnGroup(
     Masked Float=&RandomRangeFloat(Minimum=0.4, Maximum=0.8)
     Passthrough Float=&RandomRangeFloat(Minimum=0.15, Maximum=0.25, Evaluation Type=::NewEnumerator1)
    )
    Mesh Scale Max=<2 2 8>
    Mesh Scale Min=<0.5 0.5 0.5>
    Mesh Scale Mode=::NewEnumerator2
    UsePositionOffset=false
    Write Material Random=true
   )
   &AddVelocity(Velocity=&RandomRangeVector(Minimum=<-10 -10 50>, Maximum=<10 10 100>, Evaluation Type=::NewEnumerator1))
   &CameraOffset(Camera Offset Amount=&MaskFloatBySpawnGroup(
     Masked Float=&RandomRangeFloat(Minimum=-25, Maximum=25, Evaluation Type=::NewEnumerator1)
     Passthrough Float=40
    ))
   &InitialMeshOrientation(
    Mesh Orientation Mode=::NewEnumerator1
    Rotation=&MaskVectorBySpawnGroup(
     Masked Vector=&RandomRangeVector(Minimum=<0 0 0>, Maximum=<0 0 1>)
     Passthrough Vector=&RandomRangeVector(Minimum=<0 0 0>, Maximum=<1 1 1>)
    )
    Use Rotation Vector=true
   )
  ]
  particleUpdateScript [
   &ParticleState()
   &SolveForcesAndVelocity(Clamp Velocity=false, Limit Acceleration=false)
   &ScaleColor(
    Scale Alpha=&FloatFromCurve(FloatCurve=DataInterface() {propertyValues {Curve {keys [<at 0 1>,<at 1 0>]}}})
    Scale RGB=&VectorFromCurve(VectorCurve=DataInterface() {propertyValues {XCurve {keys [<at 0 0.102>,<at 0.5 0.091>]}, YCurve {keys [<at 0 0.5647>,<at 0.5 0.1113>]}, ZCurve {keys [<at 0 1>,<at 0.5 0.1198>]}}})
    ScaleA=true
    ScaleRGB=true
   )
   &ScaleMeshSize(Scale Factor=&VectorFromFloat(Value=&FloatFromCurve(FloatCurve=DataInterface() {propertyValues {Curve {keys [
          {arriveTangent=29.4334, interpMode=RCIM_Cubic, leaveTangent=29.4334, tangentMode=RCTM_User, time=0, value=0}
          {arriveTangent=15, interpMode=RCIM_Cubic, leaveTangent=15, time=0.692, value=14.6805}
          <at 1 15 RCIM_Cubic>
         ]}}})))
   &DynamicMaterialParameters(
    Index 0 Param 1=&FloatFromCurve(FloatCurve=DataInterface() {propertyValues {Curve {keys [{arriveTangent=-10.2166, interpMode=RCIM_Cubic, leaveTangent=-10.2166, tangentMode=RCTM_User, time=0, value=2}, <at 1 -2 RCIM_Cubic>]}}})
    Index 0 Param 2=&RandomRangeFloat(Minimum=2, Maximum=25)
    Index 0 Param 3=&RandomRangeFloat(Minimum=15, Maximum=100)
    Index 0 Param 4=&FloatFromCurve(FloatCurve=DataInterface() {propertyValues {Curve {keys [
         <at 0 2 RCIM_Cubic>
         {arriveTangent=-2.2566, interpMode=RCIM_Cubic, leaveTangent=-2.2566, time=0.4836, value=1.429}
         {arriveTangent=-2.7672, interpMode=RCIM_Cubic, leaveTangent=-2.7672, time=0.7173, value=0.3815}
         <at 1 0 RCIM_Cubic>
        ]}}})
    Index 1 Param 1=&RandomRangeFloat(Minimum=3, Maximum=8)
    Index 1 Param 2=&RandomRangeFloat(Minimum=1, Maximum=4)
    Index 1 Param 3=1
    Index 1 Param 4=&FloatFromCurve(FloatCurve=DataInterface() {propertyValues {Curve {keys [<at 0 -1 RCIM_Constant>,<at 0.2 -0.2417 RCIM_Constant>,<at 0.4 0.64 RCIM_Constant>,<at 0.599 1.2768 RCIM_Constant>,<at 0.8 1.8721 RCIM_Constant>,<at 1 2.5>]}}})
    Module.Index0Float0Write=true
    Module.Index0Float1Write=true
    Module.Index0Float2Write=true
    Module.Index0Float3Write=true
    Module.Index1Float0Write=true
    Module.Index1Float1Write=true
    Module.Index1Float2Write=true
    Module.Index1Float3Write=true
    Write Parameter Index 1=::NewEnumerator1
   )
   /SetVariables(
    Particles.StreakLineColor:LinearColor=&RandomRangeLinearColor(Minimum=<rgba 1.703 6.06 10 1>, Maximum=<rgba 1.661 7.155 10 1>)
    Particles.StreakLineScale:Vector3f=&Multiply_Vector(
     A=&VectorFromFloat(Value=&RandomRangeFloat(Minimum=10, Maximum=14, Evaluation Type=::NewEnumerator1))
     B=&VectorFromFloat(Value=&FloatFromCurve(FloatCurve=DataInterface() {propertyValues {Curve {keys [
           {arriveTangent=4.3887, interpMode=RCIM_Cubic, leaveTangent=4.3887, tangentMode=RCTM_User, time=0, value=0}
           {arriveTangent=1, interpMode=RCIM_Cubic, leaveTangent=1, time=0.4132, value=0.9825}
           <at 1 1>
          ]}}}))
    )
   ) {isSetParameters=true}
  ]
  rendererClasses [&NiagaraMeshRendererProperties]
  renderers [
   {rendererClass=&NiagaraMeshRendererProperties, rendererIndex=0}
   {rendererClass=&NiagaraMeshRendererProperties, rendererIndex=1}
  ]
 }
}
summary {
 systemName=NS_Figment_Snowballer_Explosion_S07
 emitters [
  {emitterName=MeshClouds, rendererClasses [&NiagaraMeshRendererProperties]}
  {emitterName=MeshCloudsSource, rendererClasses []}
  {emitterName=Initial_Burst, rendererClasses [&NiagaraMeshRendererProperties, &NiagaraLightRendererProperties]}
  {emitterName=SnowChunks, rendererClasses [&NiagaraSpriteRendererProperties]}
  {emitterName=Snowball_Mesh, rendererClasses [&NiagaraMeshRendererProperties]}
  {emitterName=Snowfall_Sprites, simTarget=GPUComputeSim, rendererClasses [&NiagaraSpriteRendererProperties]}
  {emitterName=Streaks, rendererClasses [&NiagaraMeshRendererProperties]}
  {emitterName=Smoke, rendererClasses [&NiagaraSpriteRendererProperties]}
 ]
 userVariables []
}
